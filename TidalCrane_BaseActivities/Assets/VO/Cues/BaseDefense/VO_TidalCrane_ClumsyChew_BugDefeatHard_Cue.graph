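# Sound cue for ClumsyChew's BugDefeatHard voice-over lines in the TidalCrane base activities, filed under BaseDefense.
# A Random node with equal weights picks one of eight recorded line variants.
# Playback uses the narrator VO sound class, the base activities attenuation asset and a ClumsyChew-specific concurrency set.
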SoundCue /TidalCrane_BaseActivities/Assets/VO/Cues/BaseDefense/VO_TidalCrane_ClumsyChew_BugDefeatHard_Cue
root=Random_0
nodes [
 {
  id=Random_0
  children [WavePlayer_0,WavePlayer_1,WavePlayer_2,WavePlayer_3,WavePlayer_4,WavePlayer_5,WavePlayer_6,WavePlayer_7]
  properties {weights [1,1,1,1,1,1,1,1]}
 }
 {
  id=WavePlayer_0
  properties {soundWaveAssetPtr=/TidalCrane_BaseActivities/Assets/VO/Waves/ClumsyChew/VO_TidalCrane_ClumsyChew_BugDefeatHard_010}
 }
 {
  id=WavePlayer_1
  properties {soundWaveAssetPtr=/TidalCrane_BaseActivities/Assets/VO/Waves/ClumsyChew/VO_TidalCrane_ClumsyChew_BugDefeatHard_020}
 }
 {
  id=WavePlayer_2
  properties {soundWaveAssetPtr=/TidalCrane_BaseActivities/Assets/VO/Waves/ClumsyChew/VO_TidalCrane_ClumsyChew_BugDefeatHard_030}
 }
 {
  id=WavePlayer_3
  properties {soundWaveAssetPtr=/TidalCrane_BaseActivities/Assets/VO/Waves/ClumsyChew/VO_TidalCrane_ClumsyChew_BugDefeatHard_040}
 }
 {
  id=WavePlayer_4
  properties {soundWaveAssetPtr=/TidalCrane_BaseActivities/Assets/VO/Waves/ClumsyChew/VO_TidalCrane_ClumsyChew_BugDefeatHard_050}
 }
 {
  id=WavePlayer_5
  properties {soundWaveAssetPtr=/TidalCrane_BaseActivities/Assets/VO/Waves/ClumsyChew/VO_TidalCrane_ClumsyChew_BugDefeatHard_060}
 }
 {
  id=WavePlayer_6
  properties {soundWaveAssetPtr=/TidalCrane_BaseActivities/Assets/VO/Waves/ClumsyChew/VO_TidalCrane_ClumsyChew_BugDefeatHard_070}
 }
 {
  id=WavePlayer_7
  properties {soundWaveAssetPtr=/TidalCrane_BaseActivities/Assets/VO/Waves/ClumsyChew/VO_TidalCrane_ClumsyChew_BugDefeatHard_080}
 }
]
properties {
 attenuationSettings=/TidalCrane_BaseActivities/Assets/VO/AS_Fort_BaseActivities
 maxDistance=10100
 soundClassObject=/Game/Sounds/FortSoundClassesAndModes/Fort_VO_Narrator
 volumeMultiplier=1
 concurrencySet [/TidalCrane_BaseActivities/Assets/VO/BaseActivites_VO_ClumsyChew_Conc]
}
